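# Translucent, unlit, two-sided glow material for particle effects on the Steam Power glider.
# It is enabled for Niagara mesh particles.
# Its graph is a single call to the shared MF_FX_EmissiveAdjustments function, with particle color RGB fed into the Emissive input, so the glow color comes from the particle system.
# The graph samples no textures and does not use particle alpha.
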Material /BRCosmetics/Gadgets/Assets/VinderTech_GliderChute/Glider_SteamPower/FX/M_Glider_SteamPower_ParticleGlow
graph {
 properties {
  bUsedWithNiagaraMeshParticles=true
  blendMode=BLEND_Translucent
  pixelDepthOffsetMode=PDOM_Legacy
  shadingModel=MSM_Unlit
  twoSided=true
  usedShadingModels=Unlit
 }
 nodes [MaterialFunctionCall(Emissive=ParticleColor().RGB) {materialFunction=&MF_FX_EmissiveAdjustments}]
}
